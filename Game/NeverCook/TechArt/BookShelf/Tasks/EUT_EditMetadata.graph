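# Editor utility task in the TechArt BookShelf tools that opens the Asset Metadata Manager widget for the selected asset.
# On execution it passes the first selected asset into an inline Python script.
# The script loads the widget from a hardcoded path, spawns it as a tab and calls SetCurrentAsset on it by name.
# The task then finishes without checking for an empty selection.
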EditorUtilityBlueprint /Game/NeverCook/TechArt/BookShelf/Tasks/EUT_EditMetadata
graphs [EventGraph {nodes [
   Event BeginExecution()
   > n1 = GetSelectedAssets()
   > Execute Python Script!(
    execute
    PythonScript="import unreal\r\n\r\nasset_path_string = \"/Game/Blueprints/Blutilities/ForTechArt/AssetTools/AssetMetadataManager/AssetMetadataManager_EWidget\"\r\nasset = unreal.load_asset(asset_path_string)\r\nss = unreal.get_editor_subsystem(unreal.EditorUtilitySubsystem)\r\ntab = ss.spawn_and_register_tab(asset)\r\ntab.call_method(\"SetCurrentAsset\", (asset_in,))\r\n"
    asset_in:Object Reference="Get (a copy)"(Array:Array of Object References=@n1) -> (Output:Object Reference)
   ) -> (then, ReturnValue) {inputs [asset_in]}
   > FinishExecutingTask()
  ]}]
variables []
