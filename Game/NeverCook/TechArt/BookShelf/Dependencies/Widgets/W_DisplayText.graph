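# Editor utility widget for the BookShelf tech-art tool that shows a titled, scrollable text message with a confirmation button in an editor tab.
# Update Info fills the title, body and button label and stores a tab ID, which the button passes to EditorUtilitySubsystem CloseTabByID to close the tab.
# Helper functions toggle the body's read-only state and resize the title and body fonts.
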EditorUtilityWidgetBlueprint /Game/NeverCook/TechArt/BookShelf/Dependencies/Widgets/W_DisplayText
graphs [
 EventGraph {nodes [
   "On Clicked (GotIt)"() {is=title K2Node_ComponentBoundEvent}
   > CloseTabByID!(execute, self:Editor Utility Subsystem Object Reference=Get EditorUtilitySubsystem(), NewTabID:Name=Get tabid() -> (tabid:Name=None)) -> (then, ReturnValue:bool)
   Event PreConstruct()
   Event Construct()
   Event Tick()
  ]}
 Set Read Only {nodes [
   n1 = Set Read Only!() -> (then, bReadOnly:bool=true) {is=title K2Node_FunctionEntry}
   > "SetIsReadOnly (Multi-Line Editable Text"!(
    execute
    self:"Editable Text (Multi- Line) Object Reference"=Get Body() -> (Body:Editor Utility Multi Line Editable Text Object Reference)
    bReadOnly:bool=@n1
   ) -> (then)
  ]}
 Update Info {nodes [
   n1 = Update Info!() -> (then, Title:String, Body:Text, Close Button Text:String, tab_id:Name) {is=title K2Node_FunctionEntry}
   > Set Text(Text=@n1.Body, self:"Editable Text (Multi- Line) Object Reference"=Get Body())
   > Set Text(
    Text="To Text (String)"(InString=@n1.Close Button Text)
    self:Text Object Reference=Get Confirmation() -> (Confirmation:Text Object Reference)
   )
   > Set Text(Text="To Text (String)"(InString=@n1), self=Get Title() -> (Title:Text Object Reference))
   > Set tabid(tabid:Name=@n1.tab_id) -> (Output_Get:Name=None)
  ]}
 Update Text Size {nodes [
   n1 = Update Text Size!() -> (then, Title Size:float=~, Body Size:float=~) {is=title K2Node_FunctionEntry}
   > n2 = Sequence()
   n3 = Get Title()
   > n4 = Set members in Slate Font Info!(
    execute:Exec
    StructRef:"Slate Font Info Structure (by ref)"=Get Font!(self:Text Object Reference=@n3) -> (Font:Slate Font Info Structure)
    Size:float=@n1
   ) -> (then:Exec, StructOut:"Slate Font Info Structure (by ref)") {is=title K2Node_SetFieldsInStruct}
   > SetFont(self=@n3, InFontInfo=@n4)
   n5 = Get Body()
   n6 = Set members in Slate Font Info(
    execute=@n2.then_1
    StructRef=GetFont!(self:"Editable Text (Multi- Line) Object Reference"=@n5) -> (ReturnValue:"Slate Font Info Structure (by ref)")
    Size=@n1.Body Size
   )
   > SetFont(self:"Editable Text (Multi- Line) Object Reference"=@n5, InFontInfo=@n6)
  ]}
]
variables [tabid {type=string, default=None}]
widgets {
 info {rootWidgetClass=&Border, widgetCount=10}
 widgets [
  Border_0 {
   widgetClassPath=&Border
   widget=Border_0
   properties {background {
     bIsDynamicallyLoaded=false
     drawAs=Image
     imageType=NoImage
     mirroring=NoMirror
     resourceName=None
     resourceObject=null
     tiling=NoTile
     imageSize=<32 32>
     margin=<ltrb 0 0 0 0>
     outlineSettings {
      bUseBrushTransparency=false
      roundingType=HalfHeightRadius
      width=0
      color {colorUseRule=UseColor_Specified, specifiedColor=<rgba 0 0 0 0>}
      cornerRadii=<0 0 0 0>
     }
     tintColor {colorUseRule=UseColor_Specified, specifiedColor=<rgba 0.01 0.01 0.01 1>}
     uVRegion {bIsValid=false, max=<0 0>, min=<0 0>}
    }}
  }
  VerticalBox_22 {widgetClassPath=&VerticalBox, parent=Border_0, widget=VerticalBox_22, slot {class=BorderSlot}}
  ScaleBox_1 {
   widgetClassPath=&ScaleBox
   parent=VerticalBox_22
   widget=ScaleBox_1
   slot {class=VerticalBoxSlot, properties {verticalAlignment=VAlign_Top, padding=<ltrb 10 10 10 4>}}
  }
  Border_28 {
   widgetClassPath=&Border
   parent=ScaleBox_1
   widget=Border_28
   properties {verticalAlignment=VAlign_Top, brushColor=<rgba 0.015 0.015 0.015 1>}
   slot {class=ScaleBoxSlot, properties {horizontalAlignment=HAlign_Fill, verticalAlignment=VAlign_Top}}
  }
  Title {
   widgetClassPath=&TextBlock
   bIsVariable=true
   parent=Border_28
   widget=Title
   properties {justification=Center, text=Title Text}
   slot {class=BorderSlot, properties {verticalAlignment=VAlign_Top}}
  }
  EditorUtilityScrollBox_0 {
   widgetClassPath=&EditorUtilityScrollBox
   parent=VerticalBox_22
   widget=EditorUtilityScrollBox_0
   slot {class=VerticalBoxSlot, properties {padding=<ltrb 10 10 10 10>, size {sizeRule=Fill, value=1}}}
  }
  Body {
   widgetClassPath=/Script/Blutility.EditorUtilityMultiLineEditableText
   bIsVariable=true
   parent=EditorUtilityScrollBox_0
   widget=Body
   properties {widgetStyle {
     overflowPolicy=Clip
     transformPolicy=None
     colorAndOpacity {colorUseRule=UseColor_Foreground, specifiedColor=<rgba 1 0 1 1>}
     font {
      bForceMonospaced=false
      bMaterialIsStencil=false
      fontMaterial=null
      fontObject=null
      letterSpacing=0
      monospacedWidth=1
      size=9
      skewAmount=0
      typefaceFontName=Regular
      outlineSettings {
       bApplyOutlineToDropShadows=false
       bMiteredCorners=false
       bSeparateFillAlpha=false
       outlineMaterial=null
       outlineSize=0
       outlineColor=<rgba 0 0 0 1>
      }
     }
     highlightColor {colorUseRule=UseColor_ColorTable, specifiedColor=<rgba 0 0 0 0>}
     highlightShape {
      bIsDynamicallyLoaded=false
      drawAs=Image
      imageType=NoImage
      mirroring=NoMirror
      resourceName=None
      resourceObject=null
      tiling=NoTile
      imageSize=<0 0>
      margin=<ltrb 0 0 0 0>
      outlineSettings {
       bUseBrushTransparency=false
       roundingType=HalfHeightRadius
       width=0
       color {colorUseRule=UseColor_Specified, specifiedColor=<rgba 0 0 0 0>}
       cornerRadii=<0 0 0 0>
      }
      tintColor {colorUseRule=UseColor_Specified, specifiedColor=<rgba 0.2582 0.5395 0.0685 1>}
      uVRegion {bIsValid=false, max=<0 0>, min=<0 0>}
     }
     selectedBackgroundColor {colorUseRule=UseColor_Specified, specifiedColor=<rgba 0 0.162 0.7454 1>}
     shadowColorAndOpacity=<rgba 0 0 0 1>
     shadowOffset=<0 0>
     strikeBrush {
      bIsDynamicallyLoaded=false
      drawAs=Image
      imageType=NoImage
      mirroring=NoMirror
      resourceName=None
      resourceObject=null
      tiling=NoTile
      imageSize=<32 32>
      margin=<ltrb 0 0 0 0>
      outlineSettings {
       bUseBrushTransparency=false
       roundingType=HalfHeightRadius
       width=0
       color {colorUseRule=UseColor_Specified, specifiedColor=<rgba 0 0 0 0>}
       cornerRadii=<0 0 0 0>
      }
      tintColor {colorUseRule=UseColor_Specified, specifiedColor=<rgba 1 1 1 1>}
      uVRegion {bIsValid=false, max=<0 0>, min=<0 0>}
     }
     underlineBrush {
      bIsDynamicallyLoaded=false
      drawAs=Image
      imageType=NoImage
      mirroring=NoMirror
      resourceName=None
      resourceObject=null
      tiling=NoTile
      imageSize=<32 32>
      margin=<ltrb 0 0 0 0>
      outlineSettings {
       bUseBrushTransparency=false
       roundingType=HalfHeightRadius
       width=0
       color {colorUseRule=UseColor_Specified, specifiedColor=<rgba 0 0 0 0>}
       cornerRadii=<0 0 0 0>
      }
      tintColor {colorUseRule=UseColor_Specified, specifiedColor=<rgba 1 1 1 1>}
      uVRegion {bIsValid=false, max=<0 0>, min=<0 0>}
     }
    }}
   slot {class=ScrollBoxSlot, properties {size {sizeRule=Fill, value=1}}}
  }
  Border_30 {
   widgetClassPath=&Border
   parent=VerticalBox_22
   widget=Border_30
   properties {verticalAlignment=VAlign_Bottom, brushColor=<rgba 1 1 1 0>, padding=<ltrb 4 2 4 20>}
   slot {class=VerticalBoxSlot, properties {padding=<ltrb 20 0 20 0>, size {sizeRule=Automatic, value=0.1}}}
  }
  GotIt {
   widgetClassPath=&EditorUtilityButton
   parent=Border_30
   widget=GotIt
   properties {widgetStyle {
     clickedSlateSound {resourceObject=null}
     disabled {
      bIsDynamicallyLoaded=false
      drawAs=RoundedBox
      imageType=NoImage
      mirroring=NoMirror
      resourceName=None
      resourceObject=null
      tiling=NoTile
      imageSize=<0 0>
      margin=<ltrb 0 0 0 0>
      outlineSettings {
       bUseBrushTransparency=false
       roundingType=FixedRadius
       width=1
       color {colorUseRule=UseColor_ColorTable, specifiedColor=<rgba 0 0 0 0>}
       cornerRadii=<4 4 4 4>
      }
      tintColor {colorUseRule=UseColor_Specified, specifiedColor=<rgba 0.0395 0.0395 0.0395 1>}
      uVRegion {bIsValid=false, max=<0 0>, min=<0 0>}
     }
     disabledForeground {colorUseRule=UseColor_ColorTable, specifiedColor=<rgba 0 0 0 0>}
     hovered {
      bIsDynamicallyLoaded=false
      drawAs=RoundedBox
      imageType=NoImage
      mirroring=NoMirror
      resourceName=None
      resourceObject=null
      tiling=NoTile
      imageSize=<32 32>
      margin=<ltrb 0 0 0 0>
      outlineSettings {
       bUseBrushTransparency=false
       roundingType=FixedRadius
       width=1
       color {colorUseRule=UseColor_ColorTable, specifiedColor=<rgba 0 0 0 0>}
       cornerRadii=<4 4 4 4>
      }
      tintColor {colorUseRule=UseColor_Specified, specifiedColor=<rgba 0.0953 0.0953 0.0953 1>}
      uVRegion {bIsValid=false, max=<0 0>, min=<0 0>}
     }
     hoveredForeground {colorUseRule=UseColor_ColorTable, specifiedColor=<rgba 0 0 0 0>}
     hoveredSlateSound {resourceObject=null}
     normal {
      bIsDynamicallyLoaded=false
      drawAs=RoundedBox
      imageType=NoImage
      mirroring=NoMirror
      resourceName=None
      resourceObject=null
      tiling=NoTile
      imageSize=<32 32>
      margin=<ltrb 0 0 0 0>
      outlineSettings {
       bUseBrushTransparency=false
       roundingType=FixedRadius
       width=1
       color {colorUseRule=UseColor_ColorTable, specifiedColor=<rgba 0 0 0 0>}
       cornerRadii=<1 1 1 1>
      }
      tintColor {colorUseRule=UseColor_Specified, specifiedColor=<rgba 0.0417 0.0417 0.0417 1>}
      uVRegion {bIsValid=false, max=<0 0>, min=<0 0>}
     }
     normalForeground {colorUseRule=UseColor_ColorTable, specifiedColor=<rgba 0 0 0 0>}
     normalPadding=<ltrb 12 1.5 12 1.5>
     pressed {
      bIsDynamicallyLoaded=false
      drawAs=RoundedBox
      imageType=NoImage
      mirroring=NoMirror
      resourceName=None
      resourceObject=null
      tiling=NoTile
      imageSize=<32 32>
      margin=<ltrb 0 0 0 0>
      outlineSettings {
       bUseBrushTransparency=false
       roundingType=FixedRadius
       width=1
       color {colorUseRule=UseColor_ColorTable, specifiedColor=<rgba 0 0 0 0>}
       cornerRadii=<4 4 4 4>
      }
      tintColor {colorUseRule=UseColor_Specified, specifiedColor=<rgba 0.0284 0.0284 0.0284 1>}
      uVRegion {bIsValid=false, max=<0 0>, min=<0 0>}
     }
     pressedForeground {colorUseRule=UseColor_ColorTable, specifiedColor=<rgba 0 0 0 0>}
     pressedPadding=<ltrb 12 2.5 12 0.5>
     pressedSlateSound {resourceObject=null}
    }}
   slot {class=BorderSlot, properties {verticalAlignment=VAlign_Bottom, padding=<ltrb 4 2 4 20>}}
  }
  Confirmation {
   widgetClassPath=&TextBlock
   bIsVariable=true
   parent=GotIt
   widget=Confirmation
   properties {
    justification=Center
    minDesiredWidth=0.01
    text="Got it!"
    font {
     bForceMonospaced=false
     bMaterialIsStencil=false
     fontMaterial=null
     fontObject=&Roboto
     letterSpacing=0
     monospacedWidth=1
     size=18
     skewAmount=0
     typefaceFontName=Bold
     outlineSettings {
      bApplyOutlineToDropShadows=false
      bMiteredCorners=false
      bSeparateFillAlpha=false
      outlineMaterial=null
      outlineSize=0
      outlineColor=<rgba 0 0 0 1>
     }
    }
   }
   slot {class=ButtonSlot}
  }
 ]
}
