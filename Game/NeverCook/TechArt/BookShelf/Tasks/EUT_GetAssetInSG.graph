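# Editor utility task that finds the ShotGrid/Flow pages for assets selected in the editor.
# ConformName strips variant suffixes and known prefixes from each name, then embedded Python queries the epic-shrd-techart plugin and builds a detail URL.
# Depending on Open Window?, it opens the links in a web browser or lists names and links in a W_DisplayText tab.
# It logs a warning and fails when nothing matches; an older Python lookup remains unconnected.
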EditorUtilityBlueprint /Game/NeverCook/TechArt/BookShelf/Tasks/EUT_GetAssetInSG
graphs [
 ConformName {nodes [
   n1 = ConformName!() -> (then, In:String) {is=title K2Node_FunctionEntry}
   > Set local_Name(local_Name:String=@n1) -> (Output_Get:String)
   n2 = Split(SourceString=@n1, InStr=_var_, SearchDir=FromStart)
   > n3 = Branch(Condition=@n2.ReturnValue)
   > n4 = Set local_Name(local_Name=@n2)
   n5 = Split(SourceString=Get local_Name() -> (local_Name:String), InStr=_, SearchDir=FromStart)
   n6 = Branch(execute=[@n4.then, @n3.else], Condition=Contains(SearchIn=@n5, Substring=BP, bUseCase=true))
   > n7 = Set local_Name(local_Name=@n5.RightS)
   n8 = For Each Loop(Exec=[@n7.then, @n6.else], Array=Get local_GeneralPrefixes() -> (local_GeneralPrefixes:Array of Strings))
   Return Node!(execute=@n8.Completed, Out:String=Get local_Name())
   n9 = Split(SourceString=Get local_Name(), InStr=_, SearchDir=FromStart)
   > Branch(Condition="Equal Exactly (String)"(A=@n8, B=@n9))
   > Set local_Name(local_Name=@n9.RightS)
   Return Node()
  ]}
 EventGraph {nodes [
   Event BeginExecution()
   > n1 = GetSelectedAssets()
   > n2 = Branch(Condition=Is Not Empty!(TargetArray:Array of Object References=@n1) -> (ReturnValue:bool))
   > n3 = For Each Loop(Array:Array of Object References=@n1) -> (Array Element:Object Reference)
   n4 = GetObjectName(Object=@n3)
   > Add(TargetArray=Get Names() -> (Names:Array of Strings), NewItem=@n4) {is=title K2Node_CallArrayFunction}
   > n5 = ConformName!(execute, self, In:String=@n4) -> (then, Out:String) {is=title K2Node_CallFunction}
   > n6 = Execute Python Script!(
    execute
    PythonScript="from unpipe import requires_plugin\r\nrequires_plugin(\"epic-shrd-techart\")\r\n\r\nfrom epic_shrd_techart.intg.common.sg_helpers.sg_utils import getPropSGAsset\r\n\r\ntry:\r\n    sg = getPropSGAsset(asset_name)\r\nexcept(NameError, TypeError):\r\n    path = \"\"\r\nelse:\r\n    if sg:\r\n        sgid =  sg.get('id','')\r\n        type =  sg.get('type','Asset')\r\n        path = f\"https://epicgames.shotgunstudio.com/detail/{type}/{sgid}\""
    asset_name:String=@n5
   ) -> (then, ReturnValue, path:String) {inputs [asset_name], outputs [path]}
   > n7 = Branch(Condition="Equal, Case Insensitive (String)"(A=@n6.path))
   Add(execute=@n7.else, TargetArray=Get Links() -> (Links:Array of Strings), NewItem=@n6.path)
   > Add(
    TargetArray=Get NamesPlusLinks() -> (NamesPlusLinks:Array of Strings)
    NewItem="To String (Text)"(InText=Format Text!(Format="LOCTEXT(\"{name}  --  {link}\")", name:Wildcard, link:String=@n6.path) -> (Result))
   )
   n8 = Get NamesPlusLinks()
   n9 = Branch(execute=@n3.Completed, Condition="integer > integer"(A=Length(TargetArray:Array of Strings=@n8)))
   > n10 = Branch(Condition=Get Open Window?() -> (Open Window?:bool))
   > n11 = For Each Loop(Array:Array of Strings=Get Links()) -> (Array Element:String)
   FinishExecutingTask(execute=@n11.Completed)
   Execute Python Script!(execute=@n11.LoopBody, PythonScript="import webbrowser\r\n\r\nwebbrowser.open(link)", link:String=@n11) -> (then, ReturnValue) {inputs [link]}
   n12 = SpawnAndRegisterTabAndGetID!(
    execute=@n10.else
    self:Editor Utility Subsystem Object Reference=Get EditorUtilitySubsystem()
    InBlueprint:Editor Utility Widget Blueprint Object Reference=/Game/NeverCook/TechArt/BookShelf/Dependencies/Widgets/W_DisplayText
   ) -> (then, NewTabID:Name=None, ReturnValue:Editor Utility Widget Object Reference)
   > n13 = Cast To W_DisplayText!(execute, Object=@n12.ReturnValue) -> (then, CastFailed, "As    ":"     Object Reference") {is=title K2Node_DynamicCast}
   n14 = Get NamesPlusLinks()
   > n15 = Update Info!(
    execute
    self:"     Object Reference"=@n13
    Title:String=ShotGrid / Flow Asset Paths
    Body:Text="To Text (String)"(InString=JoinStringArray(SourceArray=@n14, Separator="\r\n"))
    Close Button Text:String=Close
    tab_id:Name=@n12
   ) -> (then)
   FinishExecutingTask(execute=[@n15.then, @n13.CastFailed])
   VisLog Text!(
    execute=@n9.else
    WorldContextObject:Object Reference
    Text:String=Warning: No SG Assets found
    LogCategory:Name=EUT_GetAssetInSG
    bAddToMessageLog:bool=true
   ) -> (then)
   > FinishExecutingTask(bSuccess=false)
   FinishExecutingTask(execute=@n2.else)
   WasCancelRequested!(self:Editor Utility Task Object Reference=Self-Reference()) -> (ReturnValue:bool)
   Execute Python Script!(
    execute
    PythonScript="import unreal\r\nimport render.shared_utils as rsu\r\ntry:\r\n    sg = rsu.checkSGPropName(asset_name)\r\nexcept(NameError, TypeError):\r\n    path = \"\"\r\nelse:\r\n    sgid =  r\"/\" + sg.id\r\n    type =  r\"/\" + sg.type\r\n    path = r\"https://epicgames.shotgunstudio.com/detail\" + type + sgid"
    asset_name:Wildcard
   ) -> (then, ReturnValue, path:Wildcard) {inputs [asset_name], outputs [path]}
  ]}
]
variables [
 Links {type="array<string>", default []}
 Names {type="array<string>", default []}
 NamesPlusLinks {type="array<string>", default []}
 Open Window? {type=boolean, default=true}
]
