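# EUT_OpenToolDocumentation is an editor utility task that opens documentation in a web browser, defaulting to the BookShelf asset bookmarks Confluence page.
# On BeginExecution it runs an inline Python script that passes the Link variable to webbrowser.open.
# It then finishes the task, reporting the script node's return value as success.
# The Open Window? variable and a WasCancelRequested node are present but unused.
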EditorUtilityBlueprint /Game/NeverCook/TechArt/BookShelf/Tasks/EUT_OpenToolDocumentation
graphs [EventGraph {nodes [
   Event BeginExecution()
   > n1 = Execute Python Script!(execute, PythonScript="import webbrowser\r\n\r\nwebbrowser.open(link)", link:String=Get Link() -> (Link:String)) -> (then, ReturnValue) {inputs [link]}
   > FinishExecutingTask(bSuccess=@n1)
   WasCancelRequested!(self:Editor Utility Task Object Reference=Self-Reference()) -> (ReturnValue:bool)
  ]}]
variables [
 Link {type=string, default=https://confluence-epicgames.atlassian.net/wiki/spaces/PP/pages/151492285/BookShelf+-+Asset+Bookmarks}
 Open Window? {type=boolean, default=true}
]
